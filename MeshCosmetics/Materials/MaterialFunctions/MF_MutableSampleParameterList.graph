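# Material function that exposes a Mutable customization mask texture parameter for mesh cosmetics materials.
# It takes Material Attributes in and, when the isSetupMaterial static switch is on, overwrites Opacity with the red channel of Mutable_Customization_Mask.
# When the switch is off, the attributes pass through unchanged.
# The mask sits in the Mutable parameter group and uses the shared world-group wrap sampler.
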MaterialFunction /MeshCosmetics/Materials/MaterialFunctions/MF_MutableSampleParameterList
graph {nodes [
  n1 = FunctionInput() {inputType=FunctionInput_MaterialAttributes}
  n2 = TextureSampleParameter2D() {
   group=Mutable
   parameterName=Mutable_Customization_Mask
   samplerSource=SSM_Wrap_WorldGroupSettings
   samplerType=SAMPLERTYPE_Masks
   texture=/Engine/EngineMaterials/DefaultDiffuse_TC_Masks
  }
  FunctionOutput(None=StaticSwitchParameter(True=SetMaterialAttributes!(MaterialAttributes=@n1, Opacity=@n2.R) {attributeSetTypes [#1]}, False=@n1) {parameterName=isSetupMaterial})
 ]}
