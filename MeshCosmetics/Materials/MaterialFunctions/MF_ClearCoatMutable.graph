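# Material function for mesh cosmetics that adds a masked clear coat layer to incoming material attributes.
# A copy of the attributes gets the Clear Coat shading model, with constant values in the two custom data channels for a full-strength, low-roughness coat.
# The scalar Mask input blends between the original and coated attributes.
# A feature level switch passes the original attributes through untouched on ES3_1, skipping the effect there.
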MaterialFunction /MeshCosmetics/Materials/MaterialFunctions/MF_ClearCoatMutable
graph {nodes [
  n1 = FunctionInput() {inputType=FunctionInput_MaterialAttributes}
  n2 = SetMaterialAttributes!(
   MaterialAttributes=@n1
   ShadingModel=ShadingModel() {shadingModel=MSM_ClearCoat}
   CustomData0=Constant() {r=1}
   CustomData1=Constant() {r=0.1}
  ) {attributeSetTypes [#1, #2, #3]}
  n3 = FunctionInput() {inputName=Mask, inputType=FunctionInput_Scalar, sortPriority=3}
  FunctionOutput(None=FeatureLevelSwitch(Default=BlendMaterialAttributes(A=@n1, B=@n2, Alpha=@n3), ES3_1=@n1))
 ]}
