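# Material function that layers a cloth look onto incoming material attributes through a scalar mask, likely for mesh cosmetics.
# The default path switches to the Cloth shading model, deriving subsurface color from base color, scaling roughness and mapping the Gmap input to specular.
# A view-dependent fresnel, multiplied by a tiled fuzz texture, brightens base color at grazing angles.
# The mobile path only applies the fuzz tint to base color.
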MaterialFunction /MeshCosmetics/Materials/MaterialFunctions/MF_ClothMutable
graph {nodes [
  n1 = FunctionInput() {inputType=FunctionInput_MaterialAttributes}
  n2 = GetMaterialAttributes(None=@n1) {attributeGetTypes [#1, #2, #3]}
  n3 = Constant3Vector() {desc=Cloth Fuzz Tint, constant=<rgba 1 1 1 1>}
  n4 = Multiply(A=@n2.BaseColor, B=Constant() {desc=Cloth Subsurface Intensity, r=1})
  n5 = Constant() {desc=Cloth Roughness, r=0.9}
  n6 = Multiply(A=@n2.BaseColor, B=Constant() {desc=Cloth Base Color Intensity, r=0.9})
  n7 = DotProduct(A=PixelNormalWS(), B=CameraVectorWS())
  n8 = Clamp(None=Power(Base=OneMinus(None=@n7), Exp=Constant() {desc=Fuzz Exponent, r=4}))
  n9 = TextureSample(UVs=Multiply(A=TextureCoordinate(), B=Constant() {desc=Fuzz Tiling, r=4})) {desc=ClothFuzz Texture, samplerType=SAMPLERTYPE_LinearColor, texture=/Game/Global/Textures/Default/Masks_and_Gradients/T_Fuzz_MASK}
  n10 = Saturate(None=Multiply(A=Multiply(A=@n8, B=@n9.R), B=Constant() {desc=Fuzz Fresnel Blend, r=0.25}))
  n11 = Multiply(A=@n3, B=Add(A=Multiply(A=@n10) {constB=10}))
  n12 = Constant() {desc=Cloth Scattering Min, r=0.75}
  n13 = Constant() {desc=Cloth Scattering Max, r=1}
  n14 = FunctionInput() {inputName=Gmap, inputType=FunctionInput_Scalar}
  n15 = SetMaterialAttributes!(
   MaterialAttributes=@n2.MaterialAttributes
   ShadingModel=ShadingModel() {shadingModel=MSM_Cloth}
   SubsurfaceColor=Saturate(None=Multiply(A=@n3, B=@n4))
   Roughness=Saturate(None=Multiply(A=@n2.Roughness, B=@n5))
   CustomData0=Constant() {desc=Cloth Shader Intensity, r=1}
   Opacity=Constant() {r=1}
   BaseColor=Saturate(None=Multiply(A=@n6, B=@n11))
   Specular=Saturate(None=LinearInterpolate(A=@n12, B=@n13, Alpha=@n14))
   Normal=@n2.MaterialAttributes
  ) {attributeSetTypes [
    #4
    #5
    #3
    #6
    #7
    #1
    #8
    #2
   ]}
  n16 = SetMaterialAttributes!(MaterialAttributes=@n2.MaterialAttributes, BaseColor=Saturate(None=Multiply(A=@n2.BaseColor, B=@n3))) {attributeSetTypes [#1]}
  FunctionOutput(None=BlendMaterialAttributes(
    A=@n1
    B=ShadingPathSwitch(Default=@n15, Mobile=@n16)
    Alpha=FunctionInput() {inputName=Mask, inputType=FunctionInput_Scalar, sortPriority=3}
   ))
 ]}
