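# Material function that tints the base color of incoming material attributes with a view-angle-dependent metal reflectance lookup, for masked areas.
# The clamped dot product of camera vector and pixel normal indexes a curve atlas row parameter, defaulting to an aluminium reflectance curve.
# The curve's red channel multiplies BaseColor, and the Mask input blends between original and result before writing back to the attributes.
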MaterialFunction /MeshCosmetics/Materials/MaterialFunctions/MF_MetalLut_Mutable
graph {nodes [
  n1 = GetMaterialAttributes(None=FunctionInput() {inputName=Materials, inputType=FunctionInput_MaterialAttributes}) {attributeGetTypes [#1, #2]}
  n2 = CameraVectorWS()
  n3 = PixelNormalWS()
  n4 = CurveAtlasRowParameter(CurveTime=Clamp(None=DotProduct(A=@n2, B=@n3))) {
   atlas=&AthenaCurveAtlas01
   curve=/Game/Characters/Player/Female/Medium/Bodies/F_MED_Ms_Alpine/Textures/AluminiumreflectanceCurve
   defaultValue=189
   group=LUT
   parameterName=FresnelReflectance_LUT
  }
  n5 = Multiply(A=@n1.BaseColor, B=@n4.R)
  n6 = FunctionInput() {bUsePreviewValueAsDefault=true, inputName=Mask, inputType=FunctionInput_Vector4, sortPriority=4}
  FunctionOutput(None=SetMaterialAttributes!(MaterialAttributes=@n1.MaterialAttributes, BaseColor=LinearInterpolate(A=@n1.BaseColor, B=@n5, Alpha=@n6)) {attributeSetTypes [#2]})
 ]}
