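# Material function that applies a silk look to incoming material attributes, weighted by an Alpha input.
# A Fresnel term shifts base color toward dark grey, metallic from full to none, and roughness to half its input value from facing to grazing angles.
# The altered attributes, with shading model and CustomData0 set, are blended over the original, keeping its vertex attributes.
# On ES3_1, a feature level switch passes the input through unchanged.
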MaterialFunction /MeshCosmetics/Materials/MaterialFunctions/MF_Silk_Mutable
graph {nodes [
  n1 = FunctionInput() {inputName=Material, inputType=FunctionInput_MaterialAttributes}
  n2 = GetMaterialAttributes(None=@n1) {attributeGetTypes [
    #1
    #2
    #3
    #4
   ]}
  n3 = Constant() {r=1}
  n4 = Fresnel(ExponentIn=Constant() {r=2}) {baseReflectFraction=0}
  n5 = Constant() {r=0.5}
  n6 = SetMaterialAttributes!(
   MaterialAttributes=@n2.MaterialAttributes
   BaseColor=LinearInterpolate(A=Multiply(A=@n2.BaseColor, B=@n3), B=Constant3Vector() {constant=<rgba 0.1094 0.1094 0.1094 1>}, Alpha=@n4)
   Metallic=LinearInterpolate(A=Constant() {r=1}, B=Constant(), Alpha=@n4)
   Specular=@n2.MaterialAttributes
   Roughness=LinearInterpolate(A=@n2.Roughness, B=Multiply(A=@n2.Roughness, B=@n5), Alpha=@n4)
   ShadingModel=ShadingModel()
   SubsurfaceColor=@n2.MaterialAttributes
   CustomData0=Constant() {r=1}
  ) {attributeSetTypes [
    #1
    #2
    #3
    #4
    #5
    #6
    #7
   ]}
  n7 = FunctionInput() {inputName=Alpha, inputType=FunctionInput_Scalar, sortPriority=3}
  FunctionOutput(None=FeatureLevelSwitch(Default=BlendMaterialAttributes(A=@n2.MaterialAttributes, B=@n6, Alpha=@n7) {vertexAttributeBlendType=UseA}, ES3_1=@n1))
 ]}
