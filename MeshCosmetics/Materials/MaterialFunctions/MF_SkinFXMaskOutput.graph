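# MF_SkinFXMaskOutput is a material function that splits the color-coded SkinFX texture parameter into effect masks for cosmetic materials.
# Chroma_Key_Alpha keys red, green, blue, yellow and cyan for Cloth, ClearCoat, ThinFilm, Silk and MetalLUT, plus magenta for unlabeled Result_6.
# Each inverted key alpha is multiplied by a mask, apparently BlackMask, built from desaturated, thresholded texture color to exclude black areas.
# Keying thresholds are shared "Don't Edit" parameters.
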MaterialFunction /MeshCosmetics/Materials/MaterialFunctions/MF_SkinFXMaskOutput
graph {nodes [
  n1 = ScalarParameter() {defaultValue=0.1488, group=Don't Edit, parameterName=lumamask}
  n2 = NamedRerouteUsage()
  n3 = Constant3Vector() {constant=<rgba 1 0 0 0>}
  n4 = ScalarParameter() {group=Don't Edit, parameterName=Alpha cutoff Min}
  n5 = ScalarParameter() {defaultValue=0.15, group=Don't Edit, parameterName=Alpha cutoff Max}
  n6 = ScalarParameter() {defaultValue=1, group=Don't Edit, parameterName=Alpha Exp}
  n7 = Multiply(
   A=NamedRerouteUsage()
   B=OneMinus(None=MaterialFunctionCall!(
     LumaMask=@n1
     Image Color=@n2
     Chroma Color=@n3
     Alpha Cutoff Min=@n4
     Alpha Cutoff Max=@n5
     Alpha Exponent=@n6
     Despill Cuttoff Max
     Despill Exponent
    ) {materialFunction=/Engine/Functions/Engine_MaterialFunctions01/Opacity/Chroma_Key_Alpha}.Alpha)
  )
  FunctionOutput(None=@n7) {outputName="Red = Cloth"}
  n8 = MaterialFunctionCall(
   LumaMask=ScalarParameter() {defaultValue=0.1488, group=Don't Edit, parameterName=lumamask}
   Image Color=NamedRerouteUsage()
   Chroma Color=Constant3Vector() {constant=<rgba 0 1 0 0>}
   Alpha Cutoff Min=ScalarParameter() {group=Don't Edit, parameterName=Alpha cutoff Min}
   Alpha Cutoff Max=ScalarParameter() {defaultValue=0.15, group=Don't Edit, parameterName=Alpha cutoff Max}
   Alpha Exponent=ScalarParameter() {defaultValue=1, group=Don't Edit, parameterName=Alpha Exp}
  ) {materialFunction=&Chroma_Key_Alpha}
  FunctionOutput(None=Multiply(A=NamedRerouteUsage(), B=OneMinus(None=@n8.Alpha))) {outputName="Green = ClearCoat", sortPriority=1}
  n9 = MaterialFunctionCall(
   LumaMask=ScalarParameter() {defaultValue=0.1488, group=Don't Edit, parameterName=lumamask}
   Image Color=NamedRerouteUsage()
   Chroma Color=Constant3Vector() {constant=<rgba 0 0 1 0>}
   Alpha Cutoff Min=ScalarParameter() {group=Don't Edit, parameterName=Alpha cutoff Min}
   Alpha Cutoff Max=ScalarParameter() {defaultValue=0.15, group=Don't Edit, parameterName=Alpha cutoff Max}
   Alpha Exponent=ScalarParameter() {defaultValue=1, group=Don't Edit, parameterName=Alpha Exp}
  ) {materialFunction=&Chroma_Key_Alpha}
  FunctionOutput(None=Multiply(A=NamedRerouteUsage(), B=OneMinus(None=@n9.Alpha))) {outputName="Blue = ThinFilm", sortPriority=2}
  n10 = MaterialFunctionCall(
   LumaMask=ScalarParameter() {defaultValue=0.1488, group=Don't Edit, parameterName=lumamask}
   Image Color=NamedRerouteUsage()
   Chroma Color=Constant3Vector() {constant=<rgba 1 1 0 0>}
   Alpha Cutoff Min=ScalarParameter() {group=Don't Edit, parameterName=Alpha cutoff Min}
   Alpha Cutoff Max=ScalarParameter() {defaultValue=0.15, group=Don't Edit, parameterName=Alpha cutoff Max}
   Alpha Exponent=ScalarParameter() {defaultValue=1, group=Don't Edit, parameterName=Alpha Exp}
  ) {materialFunction=&Chroma_Key_Alpha}
  FunctionOutput(None=Multiply(A=NamedRerouteUsage(), B=OneMinus(None=@n10.Alpha))) {outputName="Yellow = Silk", sortPriority=3}
  n11 = MaterialFunctionCall(
   LumaMask=ScalarParameter() {defaultValue=0.1488, group=Don't Edit, parameterName=lumamask}
   Image Color=NamedRerouteUsage()
   Chroma Color=Constant3Vector() {constant=<rgba 0 1 1 0>}
   Alpha Cutoff Min=ScalarParameter() {group=Don't Edit, parameterName=Alpha cutoff Min}
   Alpha Cutoff Max=ScalarParameter() {defaultValue=0.15, group=Don't Edit, parameterName=Alpha cutoff Max}
   Alpha Exponent=ScalarParameter() {defaultValue=1, group=Don't Edit, parameterName=Alpha Exp}
  ) {materialFunction=&Chroma_Key_Alpha}
  FunctionOutput(None=Multiply(A=NamedRerouteUsage(), B=OneMinus(None=@n11.Alpha))) {outputName="Cyan = MetalLUT", sortPriority=4}
  n12 = MaterialFunctionCall(
   LumaMask=ScalarParameter() {defaultValue=0.1488, group=Don't Edit, parameterName=lumamask}
   Image Color=NamedRerouteUsage()
   Chroma Color=Constant3Vector() {constant=<rgba 1 0 1 0>}
   Alpha Cutoff Min=ScalarParameter() {group=Don't Edit, parameterName=Alpha cutoff Min}
   Alpha Cutoff Max=ScalarParameter() {defaultValue=0.15, group=Don't Edit, parameterName=Alpha cutoff Max}
   Alpha Exponent=ScalarParameter() {defaultValue=1, group=Don't Edit, parameterName=Alpha Exp}
  ) {materialFunction=&Chroma_Key_Alpha}
  FunctionOutput(None=Multiply(A=NamedRerouteUsage(), B=OneMinus(None=@n12.Alpha))) {outputName=Result_6, sortPriority=5}
  LinearInterpolate(B=NamedRerouteUsage(), Alpha=@n7)
  n13 = TextureSampleParameter2D() {
   desc="Red = Cloth // Green = Clear Coat // Blue = ThinFilm // Yellow = Silk // Cyan = MetalLUT"
   parameterName=SkinFX
   samplerSource=SSM_Wrap_WorldGroupSettings
   texture=&Black
  }
  n14 = Desaturation(None=@n13.RGB)
  n15 = ScalarParameter() {defaultValue=0.01, parameterName=BlackMask}
  NamedRerouteDeclaration(None=Saturate(None=Divide(A=@n14, B=@n15))) {name=BlackMask, nodeColor=<rgba 0 1 0.5313 1>}
  NamedRerouteDeclaration(None=@n13.RGB) {name=RGB, nodeColor=<rgba 0.7578 1 0 1>}
 ]}
