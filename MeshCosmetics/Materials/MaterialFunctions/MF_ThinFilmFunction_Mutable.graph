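# Material function that adds a thin-film iridescence tint to the Base Color of incoming Material Attributes, for mesh cosmetics.
# A Fresnel term drives the UVs of a spectrum texture parameter, whose halved color is added to Base Color.
# Inverse roughness times a second Fresnel masks the tint, favoring smooth, grazing surfaces, and an Alpha input blends the overall effect.
# Low quality and ES3_1 pass the input through unchanged.
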MaterialFunction /MeshCosmetics/Materials/MaterialFunctions/MF_ThinFilmFunction_Mutable
graph {nodes [
  n1 = FunctionInput() {inputName=Material Attributes, inputType=FunctionInput_MaterialAttributes}
  n2 = GetMaterialAttributes(None=@n1) {attributeGetTypes [#1, #2]}
  n3 = Multiply(A=Constant() {r=1}, B=Fresnel(ExponentIn=Constant() {r=0.5}) {baseReflectFraction=0})
  n4 = Add(
   A=@n2.BaseColor
   B=Multiply(
    A=TextureSampleParameter2D(UVs=@n3) {group=ThinFilm, parameterName=Thin Film Texture, texture=/Game/Global/Textures/Default/Thinfilm/T_ThinFilm_Spectrum_COLOR}.RGB
    B=Constant() {r=0.5}
   )
  )
  n5 = OneMinus(None=@n2.Roughness)
  n6 = Constant() {r=1}
  n7 = Multiply(A=Clamp(None=Multiply(A=@n5, B=@n6)), B=Fresnel(ExponentIn=Constant() {r=1.2164}) {baseReflectFraction=0})
  n8 = BlendMaterialAttributes(
   A=@n1
   B=SetMaterialAttributes!(MaterialAttributes=@n2.MaterialAttributes, BaseColor=LinearInterpolate(A=@n2.BaseColor, B=@n4, Alpha=@n7)) {attributeSetTypes [#2]}
   Alpha=FunctionInput() {bUsePreviewValueAsDefault=true, inputName=Alpha, inputType=FunctionInput_Scalar, sortPriority=1}
  )
  FunctionOutput(None=FeatureLevelSwitch(Default=QualitySwitch(Default=@n8, Low=@n1), ES3_1=@n1))
 ]}
