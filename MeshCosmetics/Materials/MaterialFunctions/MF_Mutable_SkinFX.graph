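# MF_Mutable_SkinFX is a material function that layers cosmetic skin surface effects onto incoming material attributes, apparently as the Mutable variant.
# MF_SkinFXMaskOutput supplies color-coded masks, one per effect.
# The attributes pass in series through clear coat (green), cloth (red), metal LUT (cyan), thin film (blue) and silk (yellow) functions.
# The cloth stage also takes a scalar Gmap input.
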MaterialFunction /MeshCosmetics/Materials/MaterialFunctions/MF_Mutable_SkinFX
graph {nodes [
  n1 = FunctionInput() {inputType=FunctionInput_MaterialAttributes}
  n2 = MaterialFunctionCall() {materialFunction=&/MF_SkinFXMaskOutput}
  n3 = MaterialFunctionCall!(
   Materials=MaterialFunctionCall!(
    Gmap=FunctionInput() {inputName=Gmap, inputType=FunctionInput_Scalar, sortPriority=3}
    In=MaterialFunctionCall!(In=@n1, Mask=@n2."Green = ClearCoat") {materialFunction=&/MF_ClearCoatMutable}.Result
    Mask=@n2."Red = Cloth"
   ) {materialFunction=&/MF_ClothMutable}.Result
   Mask=@n2."Cyan = MetalLUT"
  ) {materialFunction=&/MF_MetalLut_Mutable}
  FunctionOutput(None=MaterialFunctionCall!(
    Material=MaterialFunctionCall!(Material Attributes=@n3.Result, Alpha=@n2."Blue = ThinFilm") {materialFunction=&/MF_ThinFilmFunction_Mutable}.Result
    Alpha=@n2."Yellow = Silk"
   ) {materialFunction=&/MF_Silk_Mutable}.Result)
 ]}
